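# Post-input AnimBlueprint for the UpbeatIguanaBirch female head cosmetic, adding spring-driven secondary motion to left, right and middle dynHair bones.
# The cached input pose feeds two identically tuned chains of per-bone AnimDynamics nodes with reduced gravity and angular limits.
# The gameplay chain simulates root-relative, while the front-end chain simulates in world space.
# Each update copies bIsInFrontEnd into IsInFrontEndVis, which selects the output chain.
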AnimBlueprint /BRCosmetics/Characters/Player/Female/Medium/Heads/F_MED_UpbeatIguanaBirch_Head/Meshes/CustomSkeleton/UpbeatIguanaBirch_Head_AnimB
target_skeleton=&/UpbeatIguanaBirch_Skeleton
graphs [
 AnimGraph {nodes [
   Output Pose(Result=Blend Poses by bool(
     BlendPose_0=Use cached pose 'FrontEndDynCache'()
     BlendPose_1=Use cached pose 'GamedynCache'()
     bActiveValue=Get IsInFrontEndVis() -> (IsInFrontEndVis:bool)
    ) {blendType=Linear})
   n1 = Anim Dynamics - Bone: L_dynHair_a_01(ComponentPose=Local To Component(LocalPose=Use cached pose 'InputCache'()), Alpha=0.6, ExternalForce="0,0,0") {
    angularDampingOverride=1
    angularSpringConstant=20000
    bAngularSpring=true
    bOverrideAngularDamping=true
    bUseGravityOverride=true
    gravityScale=0.35
    simulationSpace=RootRelative
    alphaBoolBlend {blendOption=Linear}
    boundBone {boneName=L_dynHair_a_01}
    componentLinearAccScale=<1 1 1>
    componentLinearVelScale=<1 1 1>
    gravityOverride {z=-980}
    physicsBodyDefinitions [{
      boundBone {boneName=L_dynHair_a_01}
      boxExtents=<20 30 30>
      constraintSetup {angularLimitsMax=<-10 -10 -10>, angularLimitsMin=<-10 -10 -10>, angularTarget=<1 0 0>}
      localJointOffset=<3.5 0 0>
     }]
   }
   n2 = Anim Dynamics - Bone: L_dynHair_a_02(ComponentPose=@n1, Alpha=0.6, ExternalForce="0,0,0") {
    angularDampingOverride=1
    angularSpringConstant=20000
    bAngularSpring=true
    bOverrideAngularDamping=true
    bUseGravityOverride=true
    gravityScale=0.35
    simulationSpace=RootRelative
    alphaBoolBlend {blendOption=Linear}
    boundBone {boneName=L_dynHair_a_02}
    componentLinearAccScale=<1 1 1>
    componentLinearVelScale=<1 1 1>
    gravityOverride {z=-980}
    physicsBodyDefinitions [{
      boundBone {boneName=L_dynHair_a_02}
      boxExtents=<20 30 30>
      constraintSetup {angularLimitsMax=<-10 -10 -10>, angularLimitsMin=<-10 -10 -10>, angularTarget=<1 0 0>}
      localJointOffset=<3.5 0 0>
     }]
   }
   n3 = Anim Dynamics - Bone: L_dynHair_b_01(ComponentPose=@n2, Alpha=0.6, ExternalForce="0,0,0") {
    angularDampingOverride=1
    angularSpringConstant=20000
    bAngularSpring=true
    bOverrideAngularDamping=true
    bUseGravityOverride=true
    gravityScale=0.35
    simulationSpace=RootRelative
    alphaBoolBlend {blendOption=Linear}
    boundBone {boneName=L_dynHair_b_01}
    componentLinearAccScale=<1 1 1>
    componentLinearVelScale=<1 1 1>
    gravityOverride {z=-980}
    physicsBodyDefinitions [{
      boundBone {boneName=L_dynHair_b_01}
      boxExtents=<20 30 30>
      constraintSetup {angularLimitsMax=<-10 -10 -10>, angularLimitsMin=<-10 -10 -10>, angularTarget=<1 0 0>}
      localJointOffset=<3.5 0 0>
     }]
   }
   n4 = Anim Dynamics - Bone: L_dynHair_b_02(ComponentPose=@n3, Alpha=0.6, ExternalForce="0,0,0") {
    angularDampingOverride=1
    angularSpringConstant=20000
    bAngularSpring=true
    bOverrideAngularDamping=true
    bUseGravityOverride=true
    gravityScale=0.35
    simulationSpace=RootRelative
    alphaBoolBlend {blendOption=Linear}
    boundBone {boneName=L_dynHair_b_02}
    componentLinearAccScale=<1 1 1>
    componentLinearVelScale=<1 1 1>
    gravityOverride {z=-980}
    physicsBodyDefinitions [{
      boundBone {boneName=L_dynHair_b_02}
      boxExtents=<20 30 30>
      constraintSetup {angularLimitsMax=<-10 -10 -10>, angularLimitsMin=<-10 -10 -10>, angularTarget=<1 0 0>}
      localJointOffset=<3.5 0 0>
     }]
   }
   n5 = Anim Dynamics - Bone: L_dynHair_c_01(ComponentPose=@n4, Alpha=0.6, ExternalForce="0,0,0") {
    angularDampingOverride=1
    angularSpringConstant=20000
    bAngularSpring=true
    bOverrideAngularDamping=true
    bUseGravityOverride=true
    gravityScale=0.35
    simulationSpace=RootRelative
    alphaBoolBlend {blendOption=Linear}
    boundBone {boneName=L_dynHair_c_01}
    componentLinearAccScale=<1 1 1>
    componentLinearVelScale=<1 1 1>
    gravityOverride {z=-980}
    physicsBodyDefinitions [{
      boundBone {boneName=L_dynHair_c_01}
      boxExtents=<20 30 30>
      constraintSetup {angularLimitsMax=<18 18 18>, angularLimitsMin=<-18 -18 -18>, angularTarget=<1 0 0>}
      localJointOffset=<4 0 0>
     }]
   }
   n6 = Anim Dynamics - Bone: L_dynHair_c_02(ComponentPose=@n5, Alpha=0.6, ExternalForce="0,0,0") {
    angularDampingOverride=1
    angularSpringConstant=20000
    bAngularSpring=true
    bOverrideAngularDamping=true
    bUseGravityOverride=true
    gravityScale=0.35
    simulationSpace=RootRelative
    alphaBoolBlend {blendOption=Linear}
    boundBone {boneName=L_dynHair_c_02}
    componentLinearAccScale=<1 1 1>
    componentLinearVelScale=<1 1 1>
    gravityOverride {z=-980}
    physicsBodyDefinitions [{
      boundBone {boneName=L_dynHair_c_02}
      boxExtents=<20 30 30>
      constraintSetup {angularLimitsMax=<18 18 18>, angularLimitsMin=<-18 -18 -18>, angularTarget=<1 0 0>}
      localJointOffset=<4 0 0>
     }]
   }
   n7 = Anim Dynamics - Bone: L_dynHair_e_01(ComponentPose=@n6, Alpha=0.6, ExternalForce="0,0,0") {
    angularDampingOverride=1
    angularSpringConstant=20000
    bAngularSpring=true
    bOverrideAngularDamping=true
    bUseGravityOverride=true
    gravityScale=0.35
    simulationSpace=RootRelative
    alphaBoolBlend {blendOption=Linear}
    boundBone {boneName=L_dynHair_e_01}
    componentLinearAccScale=<1 1 1>
    componentLinearVelScale=<1 1 1>
    gravityOverride {z=-980}
    physicsBodyDefinitions [{
      boundBone {boneName=L_dynHair_e_01}
      boxExtents=<20 30 30>
      constraintSetup {angularLimitsMax=<18 18 18>, angularLimitsMin=<-18 -18 -18>, angularTarget=<1 0 0>}
      localJointOffset=<4 0 0>
     }]
   }
   n8 = Anim Dynamics - Bone: L_dynHair_e_02(ComponentPose=@n7, Alpha=0.6, ExternalForce="0,0,0") {
    angularDampingOverride=1
    angularSpringConstant=20000
    bAngularSpring=true
    bOverrideAngularDamping=true
    bUseGravityOverride=true
    gravityScale=0.35
    simulationSpace=RootRelative
    alphaBoolBlend {blendOption=Linear}
    boundBone {boneName=L_dynHair_e_02}
    componentLinearAccScale=<1 1 1>
    componentLinearVelScale=<1 1 1>
    gravityOverride {z=-980}
    physicsBodyDefinitions [{
      boundBone {boneName=L_dynHair_e_02}
      boxExtents=<20 30 30>
      constraintSetup {angularLimitsMax=<18 18 18>, angularLimitsMin=<-18 -18 -18>, angularTarget=<1 0 0>}
      localJointOffset=<4 0 0>
     }]
   }
   n9 = Anim Dynamics - Bone: L_dynHair_d_01(ComponentPose=@n8, Alpha=0.6, ExternalForce="0,0,0") {
    angularDampingOverride=1
    angularSpringConstant=20000
    bAngularSpring=true
    bOverrideAngularDamping=true
    bUseGravityOverride=true
    gravityScale=0.35
    simulationSpace=RootRelative
    alphaBoolBlend {blendOption=Linear}
    boundBone {boneName=L_dynHair_d_01}
    componentLinearAccScale=<1 1 1>
    componentLinearVelScale=<1 1 1>
    gravityOverride {z=-980}
    physicsBodyDefinitions [{
      boundBone {boneName=L_dynHair_d_01}
      boxExtents=<20 30 30>
      constraintSetup {angularLimitsMax=<15 15 15>, angularLimitsMin=<-15 -15 -15>, angularTarget=<1 0 0>}
      localJointOffset=<3 0 0>
     }]
   }
   n10 = Anim Dynamics - Bone: L_dynHair_d_02(ComponentPose=@n9, Alpha=0.6, ExternalForce="0,0,0") {
    angularDampingOverride=1
    angularSpringConstant=20000
    bAngularSpring=true
    bOverrideAngularDamping=true
    bUseGravityOverride=true
    gravityScale=0.35
    simulationSpace=RootRelative
    alphaBoolBlend {blendOption=Linear}
    boundBone {boneName=L_dynHair_d_02}
    componentLinearAccScale=<1 1 1>
    componentLinearVelScale=<1 1 1>
    gravityOverride {z=-980}
    physicsBodyDefinitions [{
      boundBone {boneName=L_dynHair_d_02}
      boxExtents=<20 30 30>
      constraintSetup {angularLimitsMax=<15 15 15>, angularLimitsMin=<-15 -15 -15>, angularTarget=<1 0 0>}
      localJointOffset=<3 0 0>
     }]
   }
   n11 = Anim Dynamics - Bone: R_dynHair_a_01(ComponentPose=@n10, Alpha=0.6, ExternalForce="0,0,0") {
    angularDampingOverride=1
    angularSpringConstant=20000
    bAngularSpring=true
    bOverrideAngularDamping=true
    bUseGravityOverride=true
    gravityScale=0.35
    simulationSpace=RootRelative
    alphaBoolBlend {blendOption=Linear}
    boundBone {boneName=R_dynHair_a_01}
    componentLinearAccScale=<1 1 1>
    componentLinearVelScale=<1 1 1>
    gravityOverride {z=-980}
    physicsBodyDefinitions [{
      boundBone {boneName=R_dynHair_a_01}
      boxExtents=<20 30 30>
      constraintSetup {angularLimitsMax=<15 15 15>, angularLimitsMin=<-15 -15 -15>, angularTarget=<1 0 0>}
      localJointOffset=<3 0 0>
     }]
   }
   n12 = Anim Dynamics - Bone: R_dynHair_a_02(ComponentPose=@n11, Alpha=0.6, ExternalForce="0,0,0") {
    angularDampingOverride=1
    angularSpringConstant=20000
    bAngularSpring=true
    bOverrideAngularDamping=true
    bUseGravityOverride=true
    gravityScale=0.35
    simulationSpace=RootRelative
    alphaBoolBlend {blendOption=Linear}
    boundBone {boneName=R_dynHair_a_02}
    componentLinearAccScale=<1 1 1>
    componentLinearVelScale=<1 1 1>
    gravityOverride {z=-980}
    physicsBodyDefinitions [{
      boundBone {boneName=R_dynHair_a_02}
      boxExtents=<20 30 30>
      constraintSetup {angularLimitsMax=<15 15 15>, angularLimitsMin=<-15 -15 -15>, angularTarget=<1 0 0>}
      localJointOffset=<3 0 0>
     }]
   }
   n13 = Anim Dynamics - Bone: R_dynHair_b_01(ComponentPose=@n12, Alpha=0.6, ExternalForce="0,0,0") {
    angularDampingOverride=1
    angularSpringConstant=20000
    bAngularSpring=true
    bOverrideAngularDamping=true
    bUseGravityOverride=true
    gravityScale=0.35
    simulationSpace=RootRelative
    alphaBoolBlend {blendOption=Linear}
    boundBone {boneName=R_dynHair_b_01}
    componentLinearAccScale=<1 1 1>
    componentLinearVelScale=<1 1 1>
    gravityOverride {z=-980}
    physicsBodyDefinitions [{
      boundBone {boneName=R_dynHair_b_01}
      boxExtents=<20 30 30>
      constraintSetup {angularLimitsMax=<15 15 15>, angularLimitsMin=<-15 -15 -15>, angularTarget=<1 0 0>}
      localJointOffset=<4.2 0 0>
     }]
   }
   n14 = Anim Dynamics - Bone: R_dynHair_b_02(ComponentPose=@n13, Alpha=0.6, ExternalForce="0,0,0") {
    angularDampingOverride=1
    angularSpringConstant=20000
    bAngularSpring=true
    bOverrideAngularDamping=true
    bUseGravityOverride=true
    gravityScale=0.35
    simulationSpace=RootRelative
    alphaBoolBlend {blendOption=Linear}
    boundBone {boneName=R_dynHair_b_02}
    componentLinearAccScale=<1 1 1>
    componentLinearVelScale=<1 1 1>
    gravityOverride {z=-980}
    physicsBodyDefinitions [{
      boundBone {boneName=R_dynHair_b_02}
      boxExtents=<20 30 30>
      constraintSetup {angularLimitsMax=<15 15 15>, angularLimitsMin=<-15 -15 -15>, angularTarget=<1 0 0>}
      localJointOffset=<4.2 0 0>
     }]
   }
   n15 = Anim Dynamics - Bone: R_dynHair_c_01(ComponentPose=@n14, Alpha=0.6, ExternalForce="0,0,0") {
    angularDampingOverride=1
    angularSpringConstant=20000
    bAngularSpring=true
    bOverrideAngularDamping=true
    bUseGravityOverride=true
    gravityScale=0.35
    simulationSpace=RootRelative
    alphaBoolBlend {blendOption=Linear}
    boundBone {boneName=R_dynHair_c_01}
    componentLinearAccScale=<1 1 1>
    componentLinearVelScale=<1 1 1>
    gravityOverride {z=-980}
    physicsBodyDefinitions [{
      boundBone {boneName=R_dynHair_c_01}
      boxExtents=<20 30 30>
      constraintSetup {angularLimitsMax=<15 15 15>, angularLimitsMin=<-15 -15 -15>, angularTarget=<1 0 0>}
      localJointOffset=<4.2 0 0>
     }]
   }
   n16 = Anim Dynamics - Bone: R_dynHair_c_02(ComponentPose=@n15, Alpha=0.6, ExternalForce="0,0,0") {
    angularDampingOverride=1
    angularSpringConstant=20000
    bAngularSpring=true
    bOverrideAngularDamping=true
    bUseGravityOverride=true
    gravityScale=0.35
    simulationSpace=RootRelative
    alphaBoolBlend {blendOption=Linear}
    boundBone {boneName=R_dynHair_c_02}
    componentLinearAccScale=<1 1 1>
    componentLinearVelScale=<1 1 1>
    gravityOverride {z=-980}
    physicsBodyDefinitions [{
      boundBone {boneName=R_dynHair_c_02}
      boxExtents=<20 30 30>
      constraintSetup {angularLimitsMax=<15 15 15>, angularLimitsMin=<-15 -15 -15>, angularTarget=<1 0 0>}
      localJointOffset=<4.2 0 0>
     }]
   }
   n17 = Anim Dynamics - Bone: R_dynHair_d_01(ComponentPose=@n16, Alpha=0.6, ExternalForce="0,0,0") {
    angularDampingOverride=1
    angularSpringConstant=20000
    bAngularSpring=true
    bOverrideAngularDamping=true
    bUseGravityOverride=true
    gravityScale=0.35
    simulationSpace=RootRelative
    alphaBoolBlend {blendOption=Linear}
    boundBone {boneName=R_dynHair_d_01}
    componentLinearAccScale=<1 1 1>
    componentLinearVelScale=<1 1 1>
    gravityOverride {z=-980}
    physicsBodyDefinitions [{
      boundBone {boneName=R_dynHair_d_01}
      boxExtents=<20 30 30>
      constraintSetup {angularLimitsMax=<15 15 15>, angularLimitsMin=<-15 -15 -15>, angularTarget=<1 0 0>}
      localJointOffset=<4.2 0 0>
     }]
   }
   n18 = Anim Dynamics - Bone: R_dynHair_d_02(ComponentPose=@n17, Alpha=0.6, ExternalForce="0,0,0") {
    angularDampingOverride=1
    angularSpringConstant=20000
    bAngularSpring=true
    bOverrideAngularDamping=true
    bUseGravityOverride=true
    gravityScale=0.35
    simulationSpace=RootRelative
    alphaBoolBlend {blendOption=Linear}
    boundBone {boneName=R_dynHair_d_02}
    componentLinearAccScale=<1 1 1>
    componentLinearVelScale=<1 1 1>
    gravityOverride {z=-980}
    physicsBodyDefinitions [{
      boundBone {boneName=R_dynHair_d_02}
      boxExtents=<20 30 30>
      constraintSetup {angularLimitsMax=<15 15 15>, angularLimitsMin=<-15 -15 -15>, angularTarget=<1 0 0>}
      localJointOffset=<4.2 0 0>
     }]
   }
   n19 = Anim Dynamics - Bone: R_dynHair_e_01(ComponentPose=@n18, Alpha=0.6, ExternalForce="0,0,0") {
    angularDampingOverride=1
    angularSpringConstant=20000
    bAngularSpring=true
    bOverrideAngularDamping=true
    bUseGravityOverride=true
    gravityScale=0.35
    simulationSpace=RootRelative
    alphaBoolBlend {blendOption=Linear}
    boundBone {boneName=R_dynHair_e_01}
    componentLinearAccScale=<1 1 1>
    componentLinearVelScale=<1 1 1>
    gravityOverride {z=-980}
    physicsBodyDefinitions [{
      boundBone {boneName=R_dynHair_e_01}
      boxExtents=<20 30 30>
      constraintSetup {angularLimitsMax=<15 15 15>, angularLimitsMin=<-15 -15 -15>, angularTarget=<1 0 0>}
      localJointOffset=<4.2 0 0>
     }]
   }
   n20 = Anim Dynamics - Bone: R_dynHair_e_02(ComponentPose=@n19, Alpha=0.6, ExternalForce="0,0,0") {
    angularDampingOverride=1
    angularSpringConstant=20000
    bAngularSpring=true
    bOverrideAngularDamping=true
    bUseGravityOverride=true
    gravityScale=0.35
    simulationSpace=RootRelative
    alphaBoolBlend {blendOption=Linear}
    boundBone {boneName=R_dynHair_e_02}
    componentLinearAccScale=<1 1 1>
    componentLinearVelScale=<1 1 1>
    gravityOverride {z=-980}
    physicsBodyDefinitions [{
      boundBone {boneName=R_dynHair_e_02}
      boxExtents=<20 30 30>
      constraintSetup {angularLimitsMax=<15 15 15>, angularLimitsMin=<-15 -15 -15>, angularTarget=<1 0 0>}
      localJointOffset=<4.2 0 0>
     }]
   }
   n21 = Anim Dynamics - Bone: M_dynHair_b_01(ComponentPose=@n20, Alpha=0.6, ExternalForce="0,0,0") {
    angularDampingOverride=1
    angularSpringConstant=20000
    bAngularSpring=true
    bOverrideAngularDamping=true
    bUseGravityOverride=true
    gravityScale=0.35
    simulationSpace=RootRelative
    alphaBoolBlend {blendOption=Linear}
    boundBone {boneName=M_dynHair_b_01}
    componentLinearAccScale=<1 1 1>
    componentLinearVelScale=<1 1 1>
    gravityOverride {z=-980}
    physicsBodyDefinitions [{
      boundBone {boneName=M_dynHair_b_01}
      boxExtents=<20 30 30>
      constraintSetup {angularLimitsMax=<15 15 15>, angularLimitsMin=<-15 -15 -15>, angularTarget=<1 0 0>}
      localJointOffset=<3.5 0 0>
     }]
   }
   n22 = Anim Dynamics - Bone: M_dynHair_b_02(ComponentPose=@n21, Alpha=0.6, ExternalForce="0,0,0") {
    angularDampingOverride=1
    angularSpringConstant=20000
    bAngularSpring=true
    bOverrideAngularDamping=true
    bUseGravityOverride=true
    gravityScale=0.35
    simulationSpace=RootRelative
    alphaBoolBlend {blendOption=Linear}
    boundBone {boneName=M_dynHair_b_02}
    componentLinearAccScale=<1 1 1>
    componentLinearVelScale=<1 1 1>
    gravityOverride {z=-980}
    physicsBodyDefinitions [{
      boundBone {boneName=M_dynHair_b_02}
      boxExtents=<20 30 30>
      constraintSetup {angularLimitsMax=<15 15 15>, angularLimitsMin=<-15 -15 -15>, angularTarget=<1 0 0>}
      localJointOffset=<3.5 0 0>
     }]
   }
   n23 = Anim Dynamics - Bone: M_dynHair_a_01(ComponentPose=@n22, Alpha=0.6, ExternalForce="0,0,0") {
    angularDampingOverride=1
    angularSpringConstant=20000
    bAngularSpring=true
    bOverrideAngularDamping=true
    bUseGravityOverride=true
    gravityScale=0.35
    simulationSpace=RootRelative
    alphaBoolBlend {blendOption=Linear}
    boundBone {boneName=M_dynHair_a_01}
    componentLinearAccScale=<1 1 1>
    componentLinearVelScale=<1 1 1>
    gravityOverride {z=-980}
    physicsBodyDefinitions [{
      boundBone {boneName=M_dynHair_a_01}
      boxExtents=<20 30 30>
      constraintSetup {angularLimitsMax=<15 15 15>, angularLimitsMin=<-15 -15 -15>, angularTarget=<1 0 0>}
      localJointOffset=<3.5 0 0>
     }]
   }
   n24 = Anim Dynamics - Bone: M_dynHair_a_02(ComponentPose=@n23, Alpha=0.6, ExternalForce="0,0,0") {
    angularDampingOverride=1
    angularSpringConstant=20000
    bAngularSpring=true
    bOverrideAngularDamping=true
    bUseGravityOverride=true
    gravityScale=0.35
    simulationSpace=RootRelative
    alphaBoolBlend {blendOption=Linear}
    boundBone {boneName=M_dynHair_a_02}
    componentLinearAccScale=<1 1 1>
    componentLinearVelScale=<1 1 1>
    gravityOverride {z=-980}
    physicsBodyDefinitions [{
      boundBone {boneName=M_dynHair_a_02}
      boxExtents=<20 30 30>
      constraintSetup {angularLimitsMax=<15 15 15>, angularLimitsMin=<-15 -15 -15>, angularTarget=<1 0 0>}
      localJointOffset=<3.5 0 0>
     }]
   }
   Save cached pose 'GamedynCache'(Pose=Component To Local(ComponentPose=@n24)) {cacheName=GamedynCache}
   Save cached pose 'InputCache'(Pose=Input Pose()) {cacheName=InputCache}
   n25 = Anim Dynamics - Bone: L_dynHair_a_01(ComponentPose=Local To Component(LocalPose=Use cached pose 'InputCache'()), Alpha=0.6, ExternalForce="0,0,0") {
    angularDampingOverride=1
    angularSpringConstant=20000
    bAngularSpring=true
    bOverrideAngularDamping=true
    bUseGravityOverride=true
    gravityScale=0.35
    simulationSpace=World
    alphaBoolBlend {blendOption=Linear}
    boundBone {boneName=L_dynHair_a_01}
    componentLinearAccScale=<1 1 1>
    componentLinearVelScale=<1 1 1>
    gravityOverride {z=-980}
    physicsBodyDefinitions [{
      boundBone {boneName=L_dynHair_a_01}
      boxExtents=<20 30 30>
      constraintSetup {angularLimitsMax=<-10 -10 -10>, angularLimitsMin=<-10 -10 -10>, angularTarget=<1 0 0>}
      localJointOffset=<3.5 0 0>
     }]
   }
   n26 = Anim Dynamics - Bone: L_dynHair_a_02(ComponentPose=@n25, Alpha=0.6, ExternalForce="0,0,0") {
    angularDampingOverride=1
    angularSpringConstant=20000
    bAngularSpring=true
    bOverrideAngularDamping=true
    bUseGravityOverride=true
    gravityScale=0.35
    simulationSpace=World
    alphaBoolBlend {blendOption=Linear}
    boundBone {boneName=L_dynHair_a_02}
    componentLinearAccScale=<1 1 1>
    componentLinearVelScale=<1 1 1>
    gravityOverride {z=-980}
    physicsBodyDefinitions [{
      boundBone {boneName=L_dynHair_a_02}
      boxExtents=<20 30 30>
      constraintSetup {angularLimitsMax=<-10 -10 -10>, angularLimitsMin=<-10 -10 -10>, angularTarget=<1 0 0>}
      localJointOffset=<3.5 0 0>
     }]
   }
   n27 = Anim Dynamics - Bone: L_dynHair_b_01(ComponentPose=@n26, Alpha=0.6, ExternalForce="0,0,0") {
    angularDampingOverride=1
    angularSpringConstant=20000
    bAngularSpring=true
    bOverrideAngularDamping=true
    bUseGravityOverride=true
    gravityScale=0.35
    simulationSpace=World
    alphaBoolBlend {blendOption=Linear}
    boundBone {boneName=L_dynHair_b_01}
    componentLinearAccScale=<1 1 1>
    componentLinearVelScale=<1 1 1>
    gravityOverride {z=-980}
    physicsBodyDefinitions [{
      boundBone {boneName=L_dynHair_b_01}
      boxExtents=<20 30 30>
      constraintSetup {angularLimitsMax=<-10 -10 -10>, angularLimitsMin=<-10 -10 -10>, angularTarget=<1 0 0>}
      localJointOffset=<3.5 0 0>
     }]
   }
   n28 = Anim Dynamics - Bone: L_dynHair_b_02(ComponentPose=@n27, Alpha=0.6, ExternalForce="0,0,0") {
    angularDampingOverride=1
    angularSpringConstant=20000
    bAngularSpring=true
    bOverrideAngularDamping=true
    bUseGravityOverride=true
    gravityScale=0.35
    simulationSpace=World
    alphaBoolBlend {blendOption=Linear}
    boundBone {boneName=L_dynHair_b_02}
    componentLinearAccScale=<1 1 1>
    componentLinearVelScale=<1 1 1>
    gravityOverride {z=-980}
    physicsBodyDefinitions [{
      boundBone {boneName=L_dynHair_b_02}
      boxExtents=<20 30 30>
      constraintSetup {angularLimitsMax=<-10 -10 -10>, angularLimitsMin=<-10 -10 -10>, angularTarget=<1 0 0>}
      localJointOffset=<3.5 0 0>
     }]
   }
   n29 = Anim Dynamics - Bone: L_dynHair_c_01(ComponentPose=@n28, Alpha=0.6, ExternalForce="0,0,0") {
    angularDampingOverride=1
    angularSpringConstant=20000
    bAngularSpring=true
    bOverrideAngularDamping=true
    bUseGravityOverride=true
    gravityScale=0.35
    simulationSpace=World
    alphaBoolBlend {blendOption=Linear}
    boundBone {boneName=L_dynHair_c_01}
    componentLinearAccScale=<1 1 1>
    componentLinearVelScale=<1 1 1>
    gravityOverride {z=-980}
    physicsBodyDefinitions [{
      boundBone {boneName=L_dynHair_c_01}
      boxExtents=<20 30 30>
      constraintSetup {angularLimitsMax=<18 18 18>, angularLimitsMin=<-18 -18 -18>, angularTarget=<1 0 0>}
      localJointOffset=<4 0 0>
     }]
   }
   n30 = Anim Dynamics - Bone: L_dynHair_c_02(ComponentPose=@n29, Alpha=0.6, ExternalForce="0,0,0") {
    angularDampingOverride=1
    angularSpringConstant=20000
    bAngularSpring=true
    bOverrideAngularDamping=true
    bUseGravityOverride=true
    gravityScale=0.35
    simulationSpace=World
    alphaBoolBlend {blendOption=Linear}
    boundBone {boneName=L_dynHair_c_02}
    componentLinearAccScale=<1 1 1>
    componentLinearVelScale=<1 1 1>
    gravityOverride {z=-980}
    physicsBodyDefinitions [{
      boundBone {boneName=L_dynHair_c_02}
      boxExtents=<20 30 30>
      constraintSetup {angularLimitsMax=<18 18 18>, angularLimitsMin=<-18 -18 -18>, angularTarget=<1 0 0>}
      localJointOffset=<4 0 0>
     }]
   }
   n31 = Anim Dynamics - Bone: L_dynHair_e_01(ComponentPose=@n30, Alpha=0.6, ExternalForce="0,0,0") {
    angularDampingOverride=1
    angularSpringConstant=20000
    bAngularSpring=true
    bOverrideAngularDamping=true
    bUseGravityOverride=true
    gravityScale=0.35
    simulationSpace=World
    alphaBoolBlend {blendOption=Linear}
    boundBone {boneName=L_dynHair_e_01}
    componentLinearAccScale=<1 1 1>
    componentLinearVelScale=<1 1 1>
    gravityOverride {z=-980}
    physicsBodyDefinitions [{
      boundBone {boneName=L_dynHair_e_01}
      boxExtents=<20 30 30>
      constraintSetup {angularLimitsMax=<18 18 18>, angularLimitsMin=<-18 -18 -18>, angularTarget=<1 0 0>}
      localJointOffset=<4 0 0>
     }]
   }
   n32 = Anim Dynamics - Bone: L_dynHair_e_02(ComponentPose=@n31, Alpha=0.6, ExternalForce="0,0,0") {
    angularDampingOverride=1
    angularSpringConstant=20000
    bAngularSpring=true
    bOverrideAngularDamping=true
    bUseGravityOverride=true
    gravityScale=0.35
    simulationSpace=World
    alphaBoolBlend {blendOption=Linear}
    boundBone {boneName=L_dynHair_e_02}
    componentLinearAccScale=<1 1 1>
    componentLinearVelScale=<1 1 1>
    gravityOverride {z=-980}
    physicsBodyDefinitions [{
      boundBone {boneName=L_dynHair_e_02}
      boxExtents=<20 30 30>
      constraintSetup {angularLimitsMax=<18 18 18>, angularLimitsMin=<-18 -18 -18>, angularTarget=<1 0 0>}
      localJointOffset=<4 0 0>
     }]
   }
   n33 = Anim Dynamics - Bone: L_dynHair_d_01(ComponentPose=@n32, Alpha=0.6, ExternalForce="0,0,0") {
    angularDampingOverride=1
    angularSpringConstant=20000
    bAngularSpring=true
    bOverrideAngularDamping=true
    bUseGravityOverride=true
    gravityScale=0.35
    simulationSpace=World
    alphaBoolBlend {blendOption=Linear}
    boundBone {boneName=L_dynHair_d_01}
    componentLinearAccScale=<1 1 1>
    componentLinearVelScale=<1 1 1>
    gravityOverride {z=-980}
    physicsBodyDefinitions [{
      boundBone {boneName=L_dynHair_d_01}
      boxExtents=<20 30 30>
      constraintSetup {angularLimitsMax=<15 15 15>, angularLimitsMin=<-15 -15 -15>, angularTarget=<1 0 0>}
      localJointOffset=<3 0 0>
     }]
   }
   n34 = Anim Dynamics - Bone: L_dynHair_d_02(ComponentPose=@n33, Alpha=0.6, ExternalForce="0,0,0") {
    angularDampingOverride=1
    angularSpringConstant=20000
    bAngularSpring=true
    bOverrideAngularDamping=true
    bUseGravityOverride=true
    gravityScale=0.35
    simulationSpace=World
    alphaBoolBlend {blendOption=Linear}
    boundBone {boneName=L_dynHair_d_02}
    componentLinearAccScale=<1 1 1>
    componentLinearVelScale=<1 1 1>
    gravityOverride {z=-980}
    physicsBodyDefinitions [{
      boundBone {boneName=L_dynHair_d_02}
      boxExtents=<20 30 30>
      constraintSetup {angularLimitsMax=<15 15 15>, angularLimitsMin=<-15 -15 -15>, angularTarget=<1 0 0>}
      localJointOffset=<3 0 0>
     }]
   }
   n35 = Anim Dynamics - Bone: R_dynHair_a_01(ComponentPose=@n34, Alpha=0.6, ExternalForce="0,0,0") {
    angularDampingOverride=1
    angularSpringConstant=20000
    bAngularSpring=true
    bOverrideAngularDamping=true
    bUseGravityOverride=true
    gravityScale=0.35
    simulationSpace=World
    alphaBoolBlend {blendOption=Linear}
    boundBone {boneName=R_dynHair_a_01}
    componentLinearAccScale=<1 1 1>
    componentLinearVelScale=<1 1 1>
    gravityOverride {z=-980}
    physicsBodyDefinitions [{
      boundBone {boneName=R_dynHair_a_01}
      boxExtents=<20 30 30>
      constraintSetup {angularLimitsMax=<15 15 15>, angularLimitsMin=<-15 -15 -15>, angularTarget=<1 0 0>}
      localJointOffset=<3 0 0>
     }]
   }
   n36 = Anim Dynamics - Bone: R_dynHair_a_02(ComponentPose=@n35, Alpha=0.6, ExternalForce="0,0,0") {
    angularDampingOverride=1
    angularSpringConstant=20000
    bAngularSpring=true
    bOverrideAngularDamping=true
    bUseGravityOverride=true
    gravityScale=0.35
    simulationSpace=World
    alphaBoolBlend {blendOption=Linear}
    boundBone {boneName=R_dynHair_a_02}
    componentLinearAccScale=<1 1 1>
    componentLinearVelScale=<1 1 1>
    gravityOverride {z=-980}
    physicsBodyDefinitions [{
      boundBone {boneName=R_dynHair_a_02}
      boxExtents=<20 30 30>
      constraintSetup {angularLimitsMax=<15 15 15>, angularLimitsMin=<-15 -15 -15>, angularTarget=<1 0 0>}
      localJointOffset=<3 0 0>
     }]
   }
   n37 = Anim Dynamics - Bone: R_dynHair_b_01(ComponentPose=@n36, Alpha=0.6, ExternalForce="0,0,0") {
    angularDampingOverride=1
    angularSpringConstant=20000
    bAngularSpring=true
    bOverrideAngularDamping=true
    bUseGravityOverride=true
    gravityScale=0.35
    simulationSpace=World
    alphaBoolBlend {blendOption=Linear}
    boundBone {boneName=R_dynHair_b_01}
    componentLinearAccScale=<1 1 1>
    componentLinearVelScale=<1 1 1>
    gravityOverride {z=-980}
    physicsBodyDefinitions [{
      boundBone {boneName=R_dynHair_b_01}
      boxExtents=<20 30 30>
      constraintSetup {angularLimitsMax=<15 15 15>, angularLimitsMin=<-15 -15 -15>, angularTarget=<1 0 0>}
      localJointOffset=<4.2 0 0>
     }]
   }
   n38 = Anim Dynamics - Bone: R_dynHair_b_02(ComponentPose=@n37, Alpha=0.6, ExternalForce="0,0,0") {
    angularDampingOverride=1
    angularSpringConstant=20000
    bAngularSpring=true
    bOverrideAngularDamping=true
    bUseGravityOverride=true
    gravityScale=0.35
    simulationSpace=World
    alphaBoolBlend {blendOption=Linear}
    boundBone {boneName=R_dynHair_b_02}
    componentLinearAccScale=<1 1 1>
    componentLinearVelScale=<1 1 1>
    gravityOverride {z=-980}
    physicsBodyDefinitions [{
      boundBone {boneName=R_dynHair_b_02}
      boxExtents=<20 30 30>
      constraintSetup {angularLimitsMax=<15 15 15>, angularLimitsMin=<-15 -15 -15>, angularTarget=<1 0 0>}
      localJointOffset=<4.2 0 0>
     }]
   }
   n39 = Anim Dynamics - Bone: R_dynHair_c_01(ComponentPose=@n38, Alpha=0.6, ExternalForce="0,0,0") {
    angularDampingOverride=1
    angularSpringConstant=20000
    bAngularSpring=true
    bOverrideAngularDamping=true
    bUseGravityOverride=true
    gravityScale=0.35
    simulationSpace=World
    alphaBoolBlend {blendOption=Linear}
    boundBone {boneName=R_dynHair_c_01}
    componentLinearAccScale=<1 1 1>
    componentLinearVelScale=<1 1 1>
    gravityOverride {z=-980}
    physicsBodyDefinitions [{
      boundBone {boneName=R_dynHair_c_01}
      boxExtents=<20 30 30>
      constraintSetup {angularLimitsMax=<15 15 15>, angularLimitsMin=<-15 -15 -15>, angularTarget=<1 0 0>}
      localJointOffset=<4.2 0 0>
     }]
   }
   n40 = Anim Dynamics - Bone: R_dynHair_c_02(ComponentPose=@n39, Alpha=0.6, ExternalForce="0,0,0") {
    angularDampingOverride=1
    angularSpringConstant=20000
    bAngularSpring=true
    bOverrideAngularDamping=true
    bUseGravityOverride=true
    gravityScale=0.35
    simulationSpace=World
    alphaBoolBlend {blendOption=Linear}
    boundBone {boneName=R_dynHair_c_02}
    componentLinearAccScale=<1 1 1>
    componentLinearVelScale=<1 1 1>
    gravityOverride {z=-980}
    physicsBodyDefinitions [{
      boundBone {boneName=R_dynHair_c_02}
      boxExtents=<20 30 30>
      constraintSetup {angularLimitsMax=<15 15 15>, angularLimitsMin=<-15 -15 -15>, angularTarget=<1 0 0>}
      localJointOffset=<4.2 0 0>
     }]
   }
   n41 = Anim Dynamics - Bone: R_dynHair_d_01(ComponentPose=@n40, Alpha=0.6, ExternalForce="0,0,0") {
    angularDampingOverride=1
    angularSpringConstant=20000
    bAngularSpring=true
    bOverrideAngularDamping=true
    bUseGravityOverride=true
    gravityScale=0.35
    simulationSpace=World
    alphaBoolBlend {blendOption=Linear}
    boundBone {boneName=R_dynHair_d_01}
    componentLinearAccScale=<1 1 1>
    componentLinearVelScale=<1 1 1>
    gravityOverride {z=-980}
    physicsBodyDefinitions [{
      boundBone {boneName=R_dynHair_d_01}
      boxExtents=<20 30 30>
      constraintSetup {angularLimitsMax=<15 15 15>, angularLimitsMin=<-15 -15 -15>, angularTarget=<1 0 0>}
      localJointOffset=<4.2 0 0>
     }]
   }
   n42 = Anim Dynamics - Bone: R_dynHair_d_02(ComponentPose=@n41, Alpha=0.6, ExternalForce="0,0,0") {
    angularDampingOverride=1
    angularSpringConstant=20000
    bAngularSpring=true
    bOverrideAngularDamping=true
    bUseGravityOverride=true
    gravityScale=0.35
    simulationSpace=World
    alphaBoolBlend {blendOption=Linear}
    boundBone {boneName=R_dynHair_d_02}
    componentLinearAccScale=<1 1 1>
    componentLinearVelScale=<1 1 1>
    gravityOverride {z=-980}
    physicsBodyDefinitions [{
      boundBone {boneName=R_dynHair_d_02}
      boxExtents=<20 30 30>
      constraintSetup {angularLimitsMax=<15 15 15>, angularLimitsMin=<-15 -15 -15>, angularTarget=<1 0 0>}
      localJointOffset=<4.2 0 0>
     }]
   }
   n43 = Anim Dynamics - Bone: R_dynHair_e_01(ComponentPose=@n42, Alpha=0.6, ExternalForce="0,0,0") {
    angularDampingOverride=1
    angularSpringConstant=20000
    bAngularSpring=true
    bOverrideAngularDamping=true
    bUseGravityOverride=true
    gravityScale=0.35
    simulationSpace=World
    alphaBoolBlend {blendOption=Linear}
    boundBone {boneName=R_dynHair_e_01}
    componentLinearAccScale=<1 1 1>
    componentLinearVelScale=<1 1 1>
    gravityOverride {z=-980}
    physicsBodyDefinitions [{
      boundBone {boneName=R_dynHair_e_01}
      boxExtents=<20 30 30>
      constraintSetup {angularLimitsMax=<15 15 15>, angularLimitsMin=<-15 -15 -15>, angularTarget=<1 0 0>}
      localJointOffset=<4.2 0 0>
     }]
   }
   n44 = Anim Dynamics - Bone: R_dynHair_e_02(ComponentPose=@n43, Alpha=0.6, ExternalForce="0,0,0") {
    angularDampingOverride=1
    angularSpringConstant=20000
    bAngularSpring=true
    bOverrideAngularDamping=true
    bUseGravityOverride=true
    gravityScale=0.35
    simulationSpace=World
    alphaBoolBlend {blendOption=Linear}
    boundBone {boneName=R_dynHair_e_02}
    componentLinearAccScale=<1 1 1>
    componentLinearVelScale=<1 1 1>
    gravityOverride {z=-980}
    physicsBodyDefinitions [{
      boundBone {boneName=R_dynHair_e_02}
      boxExtents=<20 30 30>
      constraintSetup {angularLimitsMax=<15 15 15>, angularLimitsMin=<-15 -15 -15>, angularTarget=<1 0 0>}
      localJointOffset=<4.2 0 0>
     }]
   }
   n45 = Anim Dynamics - Bone: M_dynHair_b_01(ComponentPose=@n44, Alpha=0.6, ExternalForce="0,0,0") {
    angularDampingOverride=1
    angularSpringConstant=20000
    bAngularSpring=true
    bOverrideAngularDamping=true
    bUseGravityOverride=true
    gravityScale=0.35
    simulationSpace=World
    alphaBoolBlend {blendOption=Linear}
    boundBone {boneName=M_dynHair_b_01}
    componentLinearAccScale=<1 1 1>
    componentLinearVelScale=<1 1 1>
    gravityOverride {z=-980}
    physicsBodyDefinitions [{
      boundBone {boneName=M_dynHair_b_01}
      boxExtents=<20 30 30>
      constraintSetup {angularLimitsMax=<15 15 15>, angularLimitsMin=<-15 -15 -15>, angularTarget=<1 0 0>}
      localJointOffset=<3.5 0 0>
     }]
   }
   n46 = Anim Dynamics - Bone: M_dynHair_b_02(ComponentPose=@n45, Alpha=0.6, ExternalForce="0,0,0") {
    angularDampingOverride=1
    angularSpringConstant=20000
    bAngularSpring=true
    bOverrideAngularDamping=true
    bUseGravityOverride=true
    gravityScale=0.35
    simulationSpace=World
    alphaBoolBlend {blendOption=Linear}
    boundBone {boneName=M_dynHair_b_02}
    componentLinearAccScale=<1 1 1>
    componentLinearVelScale=<1 1 1>
    gravityOverride {z=-980}
    physicsBodyDefinitions [{
      boundBone {boneName=M_dynHair_b_02}
      boxExtents=<20 30 30>
      constraintSetup {angularLimitsMax=<15 15 15>, angularLimitsMin=<-15 -15 -15>, angularTarget=<1 0 0>}
      localJointOffset=<3.5 0 0>
     }]
   }
   n47 = Anim Dynamics - Bone: M_dynHair_a_01(ComponentPose=@n46, Alpha=0.6, ExternalForce="0,0,0") {
    angularDampingOverride=1
    angularSpringConstant=20000
    bAngularSpring=true
    bOverrideAngularDamping=true
    bUseGravityOverride=true
    gravityScale=0.35
    simulationSpace=World
    alphaBoolBlend {blendOption=Linear}
    boundBone {boneName=M_dynHair_a_01}
    componentLinearAccScale=<1 1 1>
    componentLinearVelScale=<1 1 1>
    gravityOverride {z=-980}
    physicsBodyDefinitions [{
      boundBone {boneName=M_dynHair_a_01}
      boxExtents=<20 30 30>
      constraintSetup {angularLimitsMax=<15 15 15>, angularLimitsMin=<-15 -15 -15>, angularTarget=<1 0 0>}
      localJointOffset=<3.5 0 0>
     }]
   }
   n48 = Anim Dynamics - Bone: M_dynHair_a_02(ComponentPose=@n47, Alpha=0.6, ExternalForce="0,0,0") {
    angularDampingOverride=1
    angularSpringConstant=20000
    bAngularSpring=true
    bOverrideAngularDamping=true
    bUseGravityOverride=true
    gravityScale=0.35
    simulationSpace=World
    alphaBoolBlend {blendOption=Linear}
    boundBone {boneName=M_dynHair_a_02}
    componentLinearAccScale=<1 1 1>
    componentLinearVelScale=<1 1 1>
    gravityOverride {z=-980}
    physicsBodyDefinitions [{
      boundBone {boneName=M_dynHair_a_02}
      boxExtents=<20 30 30>
      constraintSetup {angularLimitsMax=<15 15 15>, angularLimitsMin=<-15 -15 -15>, angularTarget=<1 0 0>}
      localJointOffset=<3.5 0 0>
     }]
   }
   Save cached pose 'FrontEndDynCache'(Pose=Component To Local(ComponentPose=@n48)) {cacheName=FrontEndDynCache}
  ]}
 EventGraph {nodes [
   Event BlueprintUpdateAnimation()
   > Set IsInFrontEndVis(IsInFrontEndVis:bool=Get bIsInFrontEnd()) -> (Output_Get:bool)
   TryGetPawnOwner()
  ]}
]
variables [IsInFrontEndVis {type=boolean}]
